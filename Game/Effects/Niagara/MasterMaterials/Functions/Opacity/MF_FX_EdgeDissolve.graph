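# Material function for FX materials that dissolves a mask through leading, edge and trailing colors, outputting separate RGB and Alpha.
# Dissolve Amount offsets the mask (U gradient by default), and CheapContrast sets edge hardness.
# The result drives 3ColorBlend with only Alpha wired, whose R, G and B channels serve as weights for the three RGBA colors.
# Edge Intensity multiplies the whole edge color, alpha included, before the weighted colors are summed.
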MaterialFunction /Game/Effects/Niagara/MasterMaterials/Functions/Opacity/MF_FX_EdgeDissolve
graph {nodes [
  n1 = FunctionInput(Preview=Constant4Vector() {constant=<rgba 1 0 0 0>}.RGBA) {
   bUsePreviewValueAsDefault=true
   description=Color blended to lower values in mask
   inputName="Leading Color (RGB)"
   inputType=FunctionInput_Vector4
   sortPriority=4
  }
  n2 = FunctionInput(Preview=MaterialFunctionCall() {materialFunction=&LinearGradient}.UGradient) {bUsePreviewValueAsDefault=true, description=Mask for Dissolve, inputName=Dissolve Mask, inputType=FunctionInput_Scalar}
  n3 = FunctionInput(Preview=Constant() {r=0.5}) {
   bUsePreviewValueAsDefault=true
   description="0 -Leading, 1 - Tailing"
   inputName=Dissolve Amount
   inputType=FunctionInput_Scalar
   sortPriority=1
  }
  n4 = LinearInterpolate(Alpha=@n3) {constA=-0.5, constB=0.5}
  n5 = FunctionInput(Preview=Constant() {r=3}) {
   bUsePreviewValueAsDefault=true
   description="Dissolve Edge Hardness, Higher number is sharper edge."
   inputName=Edge Hardness
   inputType=FunctionInput_Scalar
   sortPriority=2
  }
  n6 = MaterialFunctionCall(Alpha=MaterialFunctionCall(In=Add(A=@n2, B=@n4), Contrast=@n5) {materialFunction=&CheapContrast}.Result) {materialFunction=&3ColorBlend}
  n7 = ComponentMask(None=@n6.Result) {r=true}
  n8 = FunctionInput(Preview=Constant4Vector() {constant=<rgba 0 1 0 1>}.RGBA) {
   bUsePreviewValueAsDefault=true
   description=Color associated with midtones
   inputName="Edge Color (RGBA)"
   inputType=FunctionInput_Vector4
   sortPriority=5
  }
  n9 = Multiply(
   A=@n8
   B=FunctionInput(Preview=Constant() {r=1}) {bUsePreviewValueAsDefault=true, inputName=Edge Intensity, inputType=FunctionInput_Scalar, sortPriority=3}
  )
  n10 = ComponentMask(None=@n6.Result) {g=true}
  n11 = FunctionInput(Preview=Constant4Vector() {constant=<rgba 0 0 1 1>}.RGBA) {
   bUsePreviewValueAsDefault=true
   description=Color Associated with lighter values
   inputName=Tailing Color
   inputType=FunctionInput_Vector4
   sortPriority=6
  }
  n12 = Add(A=Add(A=Multiply(A=@n1, B=@n7), B=Multiply(A=@n9, B=@n10)), B=Multiply(A=ComponentMask(None=@n6.Result) {b=true}, B=@n11))
  FunctionOutput(None=ComponentMask(None=@n12) {b=true, g=true, r=true}) {outputName=RGB}
  FunctionOutput(None=ComponentMask(None=@n12) {a=true}) {outputName=Alpha, sortPriority=1}
 ]}
