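# Translucent unlit master material for Niagara energy effects, including mesh particles, that erodes with a colored dissolve edge and an animated pattern overlay.
# One packed texture supplies scrolling UV distortion, which warps separately panned dissolve and overlay samples.
# MF_FX_EdgeDissolve builds the edge color and alpha, tinted by particle color, then blends with the overlay into MF_FX_EmissiveAdjustments.
# Dynamic parameters drive dissolve, hardness, edge intensity and overlay amount; panning can follow particle age.
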Material /Game/Effects/Niagara/MasterMaterials/Energy/M_Energy_Emissive_DissolveWithOverlay
graph {
 properties {
  bHasPixelAnimation=true
  bUsedWithNiagaraMeshParticles=true
  blendMode=BLEND_Translucent
  pixelDepthOffsetMode=PDOM_Legacy
  previewMesh=/BRCosmetics/Animation/Game/MainPlayer/Emotes/TotalFlock/FX/Meshes/SM_Emote_TotalFlock_Medallion
  shadingModel=MSM_Unlit
  usedShadingModels=Unlit
 }
 nodes [
  n1 = ParticleColor()
  n2 = TextureSampleParameter2D(UVs=NamedRerouteUsage()) {
   group=Texture
   parameterName=Disolve Pattern
   samplerType=SAMPLERTYPE_Masks
   texture=&/T_FX_Packed_UVDistortion_Erosion_Energy_01
   channelNames {a=Dissolve, b=Pattern Overlay, g=UV Distort V, r=UV Distort U}
  }
  n3 = DynamicParameter() {defaultValue=<rgba 0.5 3 1 1>, paramNames [Dissolve, Dissolve Hardness, Edge Intensity, Pattern Overlay Amount]}
  n4 = VectorParameter() {group=Edge, parameterName=Leading Color, sortPriority=33}
  n5 = VectorParameter() {group=Edge, parameterName=Edge Color, sortPriority=34, defaultValue=<rgba 0.0991 0.8658 1 1>}
  n6 = VectorParameter() {group=Edge, parameterName=Trailing Color, sortPriority=35, defaultValue=<rgba 0.0415 0.3748 1 1>}
  n7 = MaterialFunctionCall!(
   Dissolve Mask=@n2.Dissolve
   Dissolve Amount=@n3.Dissolve
   Edge Hardness=StaticSwitchParameter(True=@n3.Dissolve Hardness, False=ScalarParameter() {defaultValue=5, group=Edge, parameterName=Disolve_Hardness}) {defaultValue=true, group=Edge, parameterName=Use Particle Dissolve Hardness, sortPriority=31}
   Edge Intensity=@n3.Edge Intensity
   "Leading Color (RGB)"=AppendVector(A=@n4.RGB, B=@n4.RGB)
   "Edge Color (RGBA)"=AppendVector(A=@n5.RGB, B=@n5.RGB)
   Tailing Color=AppendVector(A=@n6.RGB, B=@n6.RGB)
  ) {materialFunction=/Game/Effects/Niagara/MasterMaterials/Functions/Opacity/MF_FX_EdgeDissolve}
  n8 = Multiply(A=@n1.RGB, B=@n7.RGB)
  n9 = VectorParameter() {group=Overlay, parameterName=Overlay Color, sortPriority=30, defaultValue=<rgba 0.0236 1.7153 2 1>}
  n10 = TextureSampleParameter2D(UVs=NamedRerouteUsage()) {
   group=Texture
   parameterName=Disolve Pattern
   samplerType=SAMPLERTYPE_Masks
   texture=&T_FX_Packed_UVDistortion_Erosion_Energy_01
   channelNames {a=Dissolve, b=Pattern Overlay, g=UV Distort V, r=UV Distort U}
  }
  n11 = Multiply(A=@n9.RGB, B=@n10.Pattern Overlay)
  n12 = DynamicParameter() {defaultValue=<rgba 0.5 3 1 1>, paramNames [Dissolve, Dissolve Hardness, Edge Intensity, Pattern Overlay Amount]}
  MaterialFunctionCall(Emissive=LinearInterpolate(A=@n8, B=LinearInterpolate(A=@n8, B=@n11, Alpha=@n10.Pattern Overlay), Alpha=@n12.Pattern Overlay Amount)) {materialFunction=&MF_FX_EmissiveAdjustments}
  Multiply(A=@n1.A, B=@n7.Alpha)
  n13 = Multiply(
   A=TextureCoordinate()
   B=AppendVector(
    A=ScalarParameter() {defaultValue=2, group=UV Distortion, parameterName=UV Distortion Tile U}
    B=ScalarParameter() {defaultValue=2, group=UV Distortion, parameterName=UV Distortion Tile V, sortPriority=33}
   )
  )
  n14 = AppendVector(
   A=ScalarParameter() {group=UV Distortion, parameterName=UV Distortion Offset U, sortPriority=34}
   B=ScalarParameter() {group=UV Distortion, parameterName=UV Distortion Offset V, sortPriority=35}
  )
  n15 = ParticleRelativeTime()
  n16 = Time() {period=1}
  n17 = ScalarParameter() {defaultValue=-0.1, group=UV Distortion, parameterName=UV Distortion Scroll U, sortPriority=36}
  n18 = ScalarParameter() {defaultValue=0.02, group=UV Distortion, parameterName=UV Distortion Scroll V, sortPriority=37}
  n19 = TextureSampleParameter2D(UVs=Panner(
    Coordinate=Add(A=@n13, B=@n14)
    Time=StaticSwitchParameter(True=@n15, False=@n16) {group=Time, parameterName=Use Particle Age}
    Speed=AppendVector(A=@n17, B=@n18)
   )) {
   group=Texture
   parameterName=Disolve Pattern
   samplerType=SAMPLERTYPE_Masks
   texture=&T_FX_Packed_UVDistortion_Erosion_Energy_01
  }
  NamedRerouteDeclaration(None=ComponentMask(None=@n19.RGB) {g=true, r=true}) {name=UV Distortion, nodeColor=<rgba 0.4062 0 1 1>}
  n20 = ScalarParameter() {defaultValue=1, group=Dissolve, parameterName=Disolve Tile U}
  n21 = ScalarParameter() {defaultValue=1, group=Dissolve, parameterName=Disolve Tile V, sortPriority=33}
  n22 = Add(
   A=Multiply(A=TextureCoordinate(), B=AppendVector(A=@n20, B=@n21))
   B=AppendVector(
    A=ScalarParameter() {group=Dissolve, parameterName=Disolve Offset U, sortPriority=34}
    B=ScalarParameter() {group=Dissolve, parameterName=Disolve Offset V, sortPriority=35}
   )
  )
  n23 = StaticSwitchParameter(True=ParticleRelativeTime(), False=Time() {period=1}) {group=Time, parameterName=Use Particle Age}
  n24 = AppendVector(
   A=ScalarParameter() {defaultValue=0.01, group=Dissolve, parameterName=Disolve Scroll U, sortPriority=36}
   B=ScalarParameter() {defaultValue=0.01, group=Dissolve, parameterName=Disolve Scroll V, sortPriority=37}
  )
  n25 = NamedRerouteUsage()
  n26 = ScalarParameter() {defaultValue=0.1, group=Dissolve, parameterName=Dissolve UV Distortion Stength, sortPriority=31}
  NamedRerouteDeclaration(None=Add(A=Panner(Coordinate=@n22, Time=@n23, Speed=@n24), B=Multiply(A=@n25, B=@n26))) {name=UVs Dissolve, nodeColor=<rgba 0 1 0.0859 1>}
  n27 = ScalarParameter() {defaultValue=0.1, group=Overlay, parameterName=Overlay Tile U}
  n28 = ScalarParameter() {defaultValue=1, group=Overlay, parameterName=Overlay Tile V, sortPriority=33}
  n29 = Add(
   A=Multiply(A=TextureCoordinate(), B=AppendVector(A=@n27, B=@n28))
   B=AppendVector(
    A=ScalarParameter() {group=Overlay, parameterName=Overlay Offset U, sortPriority=34}
    B=ScalarParameter() {group=Overlay, parameterName=Overlay Offset V, sortPriority=35}
   )
  )
  n30 = StaticSwitchParameter(True=ParticleRelativeTime(), False=Time() {period=1}) {group=Time, parameterName=Use Particle Age}
  n31 = AppendVector(
   A=ScalarParameter() {defaultValue=0.1, group=Overlay, parameterName=Overlay Scroll U, sortPriority=36}
   B=ScalarParameter() {group=Overlay, parameterName=Overlay Scroll V, sortPriority=37}
  )
  n32 = NamedRerouteUsage()
  n33 = ScalarParameter() {defaultValue=0.2, group=Overlay, parameterName=Overlay UV Distortion Stength, sortPriority=31}
  NamedRerouteDeclaration(None=Add(A=Panner(Coordinate=@n29, Time=@n30, Speed=@n31), B=Multiply(A=@n32, B=@n33))) {name=UVs Overlay, nodeColor=<rgba 0 1 0.0859 1>}
 ]
}
