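# Additive, unlit, two-sided master material for energy effects on Niagara sprites, ribbons and mesh particles.
# Panning noise warps the emissive and dissolve UVs; emissive, refined and optionally curve-atlas tinted, drives color.
# Emissive times mask forms opacity, shaped by optional diamond gradient, dissolve and fresnel, plus depth and camera fades.
# World position offset blends sine wave, spline thickening and trailing bend, with dynamic parameters supplying per-particle control.
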Material /Game/Effects/Niagara/MasterMaterials/Energy/M_Energy_Master_NoDistortion
graph {
 properties {
  bAutomaticallySetUsageInEditor=false
  bUseTranslucencyVertexFog=false
  bUsedWithNiagaraMeshParticles=true
  bUsedWithNiagaraRibbons=true
  bUsedWithNiagaraSprites=true
  bUsedWithStaticMesh=false
  blendMode=BLEND_Additive
  pixelDepthOffsetMode=PDOM_Legacy
  shadingModel=MSM_Unlit
  twoSided=true
  usedShadingModels=Unlit
 }
 nodes [
  n1 = ParticleColor()
  n2 = ScalarParameter() {parameterName=Use On Rigged Mesh, sliderMax=1, sortPriority=0}
  n3 = LinearInterpolate(
   A=TextureCoordinate()
   B=TextureCoordinate() {coordinateIndex=1}
   Alpha=ScalarParameter() {group="Noise ", parameterName="Use Noise TexCoord [0] or [1]"}
  )
  n4 = VectorParameter() {group="Noise ", parameterName=Noise UV, defaultValue=<rgba 1 1 0 0>}
  n5 = MaterialFunctionCall(X=@n4.R, Y=@n4.R) {materialFunction=&MakeFloat2}
  n6 = MaterialFunctionCall(X=@n4.B, Y=@n4.B) {materialFunction=&MakeFloat2}
  n7 = DynamicParameter() {parameterIndex=3, defaultValue=<rgba 0 0 1 1>, paramNames [UVTilingRandom, TimeOffsetRandom, SplineWidthBase, SplineWidthTip]}
  n8 = Add(A=Multiply(A=Time(), B=ScalarParameter() {defaultValue=1, parameterName=NoisePanningSpeed}), B=@n7.TimeOffsetRandom)
  n9 = TextureSample(
   UVs=Add(A=Multiply(A=@n3, B=@n5.Result), B=Multiply(A=@n6.Result, B=@n8))
   Tex=TextureObjectParameter() {group="Noise ", parameterName=Noise Tex}
  )
  n10 = ScalarParameter() {defaultValue=-1, group="Noise ", parameterName=Noise Strength}
  n11 = DynamicParameter() {defaultValue=<rgba 0 1 1 1>, paramNames [Emi V Offset, Emi V Tiling, Noise, Dissolve]}
  n12 = Add(
   A=Multiply(A=ComponentMask(None=@n9.RGB) {g=true, r=true}, B=Multiply(A=@n10, B=@n11.Noise))
   B=ScalarParameter() {group="Noise ", parameterName=Noise Offset}
  )
  n13 = VectorParameter() {
   group=Emissive
   parameterName=Emissive TilingMethod
   sortPriority=3
   channelNames {a=Use Dynamic Offset, b=Rotation, g=V Clamp, r=U Clamp}
   defaultValue=<rgba 0 1 0 0>
  }
  n14 = MaterialFunctionCall(
   UVs=LinearInterpolate(
    A=TextureCoordinate()
    B=TextureCoordinate() {coordinateIndex=1}
    Alpha=ScalarParameter() {group=Emissive, parameterName="Use Emissive TexCoord [0] or [1]"}
   )
   "Rotation Angle (0-1)"=@n13.Rotation
  ) {materialFunction=&CustomRotator}
  n15 = VectorParameter() {group=Emissive, parameterName=Emissive UV, sortPriority=2, defaultValue=<rgba 1 1 0 0>}
  n16 = DynamicParameter() {defaultValue=<rgba 0 1 1 1>, paramNames [Emi V Offset, Emi V Tiling, Noise, Dissolve]}
  n17 = MaterialFunctionCall(X=@n15.R, Y=Multiply(A=@n15.G, B=@n16.Emi V Tiling)) {materialFunction=&MakeFloat2}
  n18 = MaterialFunctionCall(X=@n15.B, Y=LinearInterpolate(A=@n15.A, B=@n16.Emi V Offset, Alpha=@n13.Use Dynamic Offset)) {materialFunction=&MakeFloat2}
  n19 = Time()
  n20 = ScalarParameter() {defaultValue=1, parameterName=EmissivePanningSpeed}
  n21 = DynamicParameter() {parameterIndex=3, defaultValue=<rgba 0 0 1 1>, paramNames [UVTilingRandom, TimeOffsetRandom, SplineWidthBase, SplineWidthTip]}
  n22 = Multiply(A=@n18.Result, B=Add(A=Multiply(A=@n19, B=@n20), B=@n21.TimeOffsetRandom))
  n23 = MaterialFunctionCall(X=@n13.U Clamp, Y=@n13.U Clamp) {materialFunction=&MakeFloat2}
  n24 = Add(A=@n12, B=Add(A=Multiply(A=@n14.Rotated Values, B=@n17.Result), B=LinearInterpolate(A=@n22, B=@n18.Result, Alpha=@n23.Result)))
  n25 = TextureObjectParameter() {group=Emissive, parameterName=Emissive Tex, sortPriority=0}
  n26 = Multiply(
   A=MaterialFunctionCall!(TexCoord=@n24, ClampChannel=@n23.Result) {materialFunction=/Game/Effects/Fort_Effects/MaterialFunctions/MF_FX_TilingMethod}.Result
   B=ChannelMaskParameter(None=TextureSample(UVs=@n24, Tex=@n25).RGBA) {group=Emissive, parameterName=Emissive Channel, sortPriority=1}.R
  )
  n27 = VectorParameter() {
   group=Mask
   parameterName=Mask TilingMethod
   sortPriority=2
   channelNames {a="Use TexCoord [0] or [1]", b=Rotation, g=V Clamp, r=U Clamp}
   defaultValue=<rgba 1 1 0 0>
  }
  n28 = LinearInterpolate(A=TextureCoordinate(), B=TextureCoordinate() {coordinateIndex=1}, Alpha=@n27."Use TexCoord [0] or [1]")
  n29 = VectorParameter() {group=Mask, parameterName=Mask UV, sortPriority=1, defaultValue=<rgba 1 1 0 0>}
  n30 = DynamicParameter() {parameterIndex=2, defaultValue=<rgba 0 0 1 0>, paramNames [Sine Speed, Sine Tile, Sine Strength, Param4]}
  n31 = LinearInterpolate(A=@n29.G, B=Multiply(A=@n29.G, B=@n30.Param4), Alpha=ScalarParameter() {parameterName=UseDPtoAnimateMaskVtile})
  n32 = MaterialFunctionCall(X=@n29.B, Y=@n29.B) {materialFunction=&MakeFloat2}
  n33 = DynamicParameter() {parameterIndex=3, defaultValue=<rgba 0 0 1 1>, paramNames [UVTilingRandom, TimeOffsetRandom, SplineWidthBase, SplineWidthTip]}
  n34 = Add(A=Time(), B=@n33.TimeOffsetRandom)
  n35 = MaterialFunctionCall(X=@n27.U Clamp, Y=@n27.U Clamp) {materialFunction=&MakeFloat2}
  n36 = Add(
   A=Multiply(
    A=MaterialFunctionCall(UVs=@n28, "Rotation Angle (0-1)"=@n27.Rotation) {materialFunction=&CustomRotator}.Rotated Values
    B=MaterialFunctionCall(X=@n29.R, Y=@n31) {materialFunction=&MakeFloat2}.Result
   )
   B=LinearInterpolate(A=Multiply(A=@n32.Result, B=@n34), B=@n32.Result, Alpha=@n35.Result)
  )
  n37 = MaterialFunctionCall(TexCoord=@n36, ClampChannel=@n35.Result) {materialFunction=&MF_FX_TilingMethod}
  n38 = TextureSample(UVs=@n36, Tex=TextureObjectParameter() {group=Mask, parameterName=Mask Tex, sortPriority=0})
  n39 = Multiply(
   A=Multiply(A=@n26, B=ScalarParameter() {defaultValue=1, group=Emissive, parameterName=Emissive Mask Multiply, sortPriority=5})
   B=Multiply(A=Multiply(A=@n37.Result, B=@n38.R), B=ScalarParameter() {defaultValue=1, group=Mask, parameterName=Mask Multiply, sortPriority=3})
  )
  n40 = ScalarParameter() {defaultValue=1, group=DiamondGradient, parameterName=DiamondGradient Falloff, sortPriority=5}
  n41 = LinearInterpolate(
   A=@n39
   B=Multiply(
    A=@n39
    B=MaterialFunctionCall!(Falloff=@n40) {materialFunction=/Engine/Functions/Engine_MaterialFunctions01/Gradient/DiamondGradient}.DiamondGradient
   )
   Alpha=ScalarParameter() {group=DiamondGradient, parameterName=Use DiamondGradient, sortPriority=4}
  )
  n42 = VectorParameter() {
   group=Dissolve
   parameterName=Dissolve Control2
   sortPriority=2
   channelNames {a="", b=Tex Multiply, g=Use Position Dissolve, r=Noise Strength}
   defaultValue=<rgba 0 0 0.5 0>
  }
  n43 = VectorParameter() {
   group=Dissolve
   parameterName=Dissolve Control
   sortPriority=1
   channelNames {a=Intensity, b=Smooth, g=Use Pan of time, r="Use TexCoord [0] or [1]"}
   defaultValue=<rgba 0 0 3 1>
  }
  n44 = VectorParameter() {group=Dissolve, parameterName=Dissolve UV, sortPriority=3, defaultValue=<rgba 1 1 0 0>}
  n45 = Multiply(
   A=LinearInterpolate(A=TextureCoordinate(), B=TextureCoordinate() {coordinateIndex=1}, Alpha=@n43."Use TexCoord [0] or [1]")
   B=MaterialFunctionCall(X=@n44.R, Y=@n44.R) {materialFunction=&MakeFloat2}.Result
  )
  n46 = MaterialFunctionCall(X=@n44.B, Y=@n44.B) {materialFunction=&MakeFloat2}
  n47 = Time()
  n48 = DynamicParameter() {parameterIndex=3, defaultValue=<rgba 0 0 1 1>, paramNames [UVTilingRandom, TimeOffsetRandom, SplineWidthBase, SplineWidthTip]}
  n49 = LinearInterpolate(A=@n46.Result, B=Multiply(A=@n46.Result, B=Add(A=@n47, B=@n48.TimeOffsetRandom)), Alpha=@n43.Use Pan of time)
  n50 = TextureSample(
   UVs=Add(A=Multiply(A=@n12, B=@n42.Noise Strength), B=Add(A=@n45, B=@n49))
   Tex=TextureObjectParameter() {group=Dissolve, parameterName=Dissolve Tex, sortPriority=2}
  )
  n51 = Multiply(A=@n50.R, B=@n42.Tex Multiply)
  n52 = TextureCoordinate()
  n53 = LinearInterpolate(
   A=ComponentMask(None=@n52) {r=true}
   B=ComponentMask(None=@n52) {g=true}
   Alpha=ScalarParameter() {desc="0=R; 1=G", group=Base, parameterName=Use TexCoord R or G, sliderMax=1, sortPriority=1}
  )
  n54 = LinearInterpolate(A=@n53, B=OneMinus(None=@n53), Alpha=ScalarParameter() {group=Base, parameterName=Use TexCoord R or G Invert, sliderMax=1, sortPriority=1})
  n55 = LinearInterpolate(A=@n50.R, B=Saturate(None=Add(A=@n51, B=@n54)), Alpha=@n42.Use Position Dissolve)
  n56 = Multiply(A=DynamicParameter() {defaultValue=<rgba 0 1 1 1>, paramNames [Emi V Offset, Emi V Tiling, Noise, Dissolve]}.Dissolve, B=@n43.Intensity)
  n57 = Saturate(None=Subtract(A=Multiply(A=@n55, B=@n43.Smooth), B=LinearInterpolate(A=@n43.Smooth, Alpha=@n56) {constB=-1}))
  n58 = Multiply(
   A=Add(A=@n1.A, B=@n2)
   B=LinearInterpolate(A=@n41, B=Multiply(A=@n41, B=@n57), Alpha=ScalarParameter() {group=Dissolve, parameterName=Use Dissolve, sortPriority=0})
  )
  n59 = PixelNormalWS()
  n60 = VertexNormalWS()
  n61 = QualitySwitch(Default=@n59, Low=@n60, High=@n59, Medium=@n60, Epic=@n59)
  n62 = Abs(None=DotProduct(A=CameraVectorWS(), B=ShadingPathSwitch(Default=@n61, Deferred=@n61, Forward=@n60, Mobile=@n60)))
  n63 = VectorParameter() {
   group=Base
   parameterName=Mesh Fresnel Fade
   sortPriority=1
   channelNames {a="", b=Power, g=Invert, r=Use Mesh Fresnel Fade}
   defaultValue=<rgba 0 0 1 0>
  }
  n64 = Power(Base=LinearInterpolate(A=@n62, B=OneMinus(None=@n62), Alpha=@n63.Invert), Exp=@n63.Power)
  n65 = LinearInterpolate(A=@n58, B=Multiply(A=Saturate(None=@n64), B=@n58), Alpha=@n63.Use Mesh Fresnel Fade)
  n66 = DepthFade(Opacity=@n65, FadeDistance=ScalarParameter() {defaultValue=100, group=Base, parameterName=Depth Fade})
  n67 = ShadingPathSwitch(
   Default=@n66
   Deferred=@n66
   Forward=@n65
   Mobile=LinearInterpolate(A=@n65, B=@n66, Alpha=ScalarParameter() {group=Base, parameterName=Use Depth Fade On Switch})
  )
  n68 = Multiply(
   A=MaterialFunctionCall(Distance=ScalarParameter() {defaultValue=256, group=Base, parameterName=Camera Fade Distance}) {materialFunction=&Near_Camera_Fade}.Result
   B=QualitySwitch(Default=@n67, Low=@n65, High=@n67, Medium=@n67, Epic=@n67)
  )
  n69 = CurveAtlasRowParameter(CurveTime=@n54) {
   atlas=/Game/Athena/LinearColorCurves/VFX_CurveAtlas
   curve=/Game/Weapons/FORT_Melee/Pickaxe_Cube_Ninja_Male/FX/Curve_CubeNinja_Pickaxe_Color
   defaultValue=10
   group=Curve
   parameterName=Curve
  }
  LinearInterpolate(A=@n68, B=Multiply(A=@n68, B=@n69.A), Alpha=ScalarParameter() {group=Curve, parameterName=Use Curve Alpha})
  n70 = DynamicParameter() {parameterIndex=3, defaultValue=<rgba 0 0 1 1>, paramNames [UVTilingRandom, TimeOffsetRandom, SplineWidthBase, SplineWidthTip]}
  n71 = Add(A=Time(), B=@n70.TimeOffsetRandom)
  n72 = VectorParameter() {group=WPO, parameterName=WPO Sine, channelNames {a="", b=Sine Strength, g=Sine Tile, r=Sine Speed}, defaultValue=<rgba 2 2 1 1>}
  n73 = DynamicParameter() {parameterIndex=2, defaultValue=<rgba 0 0 1 0>, paramNames [Sine Speed, Sine Tile, Sine Strength, Param4]}
  n74 = Add(A=@n72.Sine Speed, B=@n73.Sine Speed)
  n75 = Add(A=@n70.UVTilingRandom, B=@n54)
  n76 = Add(A=@n72.Sine Tile, B=@n73.Sine Tile)
  n77 = Sine(None=Add(A=Multiply(A=@n71, B=@n74), B=Multiply(A=@n75, B=@n76) {constB=2}))
  n78 = Multiply(
   A=MaterialFunctionCall(X=@n77, Y=@n77, Z=@n77) {materialFunction=&MakeFloat3}.Result
   B=@n72.Sine Strength
  ) {constB=0.2}
  n79 = VectorParameter() {group=WPO, parameterName=WPO Sine Axis, channelNames {a="", b=Z, g=Y, r=X}, defaultValue=<rgba 1 1 1 0>}
  n80 = LinearInterpolate(
   A=OneMinus(None=@n54)
   B=Sine(None=Multiply(A=@n54) {constB=0.5})
   Alpha=ScalarParameter() {group=WPO, parameterName=Use WPO Center Mask, sliderMax=1, sortPriority=0}
  )
  n81 = Multiply(A=VertexNormalWS(), B=LinearInterpolate(A=Constant3Vector(), B=Multiply(A=@n78, B=@n79), Alpha=@n80))
  n82 = DynamicParameter() {parameterIndex=2, defaultValue=<rgba 0 0 1 0>, paramNames [Sine Speed, Sine Tile, Sine Strength, Param4]}
  n83 = VectorParameter() {group=WPO, parameterName=WPO Switch Control, sortPriority=0, channelNames {a="", b=Use TrailingMode, g=Use SplineThicken, r=Use SineWave}}
  n84 = LinearInterpolate(A=Constant(), B=Transform(None=Multiply(A=@n81, B=@n82.Sine Strength) {constB=10}), Alpha=@n83.Use SineWave)
  n85 = VectorParameter() {group=WPO, parameterName=Spline Width, defaultValue=<rgba 10 10 0 0>}
  n86 = DynamicParameter() {parameterIndex=3, defaultValue=<rgba 0 0 1 1>, paramNames [UVTilingRandom, TimeOffsetRandom, SplineWidthBase, SplineWidthTip]}
  n87 = Multiply(A=@n85.R, B=@n86.SplineWidthBase)
  n88 = Multiply(A=@n85.G, B=@n86.SplineWidthTip)
  n89 = LinearInterpolate(
   A=@n84
   B=Add(
    A=@n84
    B=MaterialFunctionCall(WidthBase=@n87, WidthTip=@n88) {materialFunction=&SplineThicken}.WorldPositionOffset
   )
   Alpha=@n83.Use SplineThicken
  )
  n90 = DynamicParameter() {desc=Parameters set by Niagara, parameterIndex=1, paramNames [X,Y,Z,WPOIntensity]}
  n91 = VectorParameter() {group=WPO, parameterName=WPO Control, channelNames {a="", b=BendPower, g=Max, r=Min}, defaultValue=<rgba -50 50 3 0>}
  n92 = Clamp(
   None=Subtract(
    A=MaterialFunctionCall(X=@n90.X, Y=@n90.X, Z=@n90.X) {materialFunction=&MakeFloat3}.Result
    B=ObjectPositionWS()
   )
   Min=@n91.Min
   Max=@n91.Min
  ) {maxDefault=50, minDefault=-50}
  n93 = Multiply(A=Power(Base=@n80, Exp=@n91.BendPower), B=@n90.WPOIntensity)
  LinearInterpolate(A=@n89, B=Add(A=LinearInterpolate(B=@n92, Alpha=@n93), B=@n89), Alpha=@n83.Use TrailingMode)
  n94 = VectorParameter() {group=Emissive, parameterName=Emissive Refine, sortPriority=4, defaultValue=<rgba 1 1 1 0.5>}
  n95 = MaterialFunctionCall(Texture=@n26, RGB=@n94.RGB, A=@n94.RGB) {materialFunction=&MF_FX_Refine}
  n96 = Multiply(A=@n69, B=@n95.Result)
  n97 = ScalarParameter() {group=Curve, parameterName=Use Curve Color, sortPriority=0}
  MaterialFunctionCall(Emissive=Multiply(A=LinearInterpolate(A=@n95.Result, B=@n96, Alpha=@n97), B=Add(A=@n1.RGB, B=@n2))) {materialFunction=&MF_FX_EmissiveAdjustments}
 ]
}
