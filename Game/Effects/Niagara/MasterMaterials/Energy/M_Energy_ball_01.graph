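# Unlit translucent Niagara material for an animated energy-ball effect on mesh particles and sprites.
# Panning smoke and noise layers, a Fresnel term and a per-particle randomized smoke layer combine into particle-colored emissive.
# Opacity multiplies a panning mask, noise, particle alpha and view-angle fade, with depth fade skipped on forward and mobile paths.
# Randomly offset blurred noise scales vertex normals for surface wobble.
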Material /Game/Effects/Niagara/MasterMaterials/Energy/M_Energy_ball_01
graph {
 properties {
  bAutomaticallySetUsageInEditor=false
  bUsedWithNiagaraMeshParticles=true
  bUsedWithNiagaraSprites=true
  blendMode=BLEND_Translucent
  shadingModel=MSM_Unlit
  usedShadingModels=Unlit
 }
 nodes [
  n1 = Panner(Coordinate=TextureCoordinate() {vTiling=2}) {speedX=-1, speedY=0.5}
  n2 = TextureCoordinate() {vTiling=0.8}
  n3 = Multiply(
   A=LinearInterpolate(Alpha=TextureSample(UVs=@n1) {texture=&T_Inky_Smoke_Tile}.R) {constB=5}
   B=TextureSample(UVs=Panner(Coordinate=@n2) {speedX=-0.5, speedY=0.3}) {texture=/Game/Effects/Niagara/MasterMaterials/Energy/Texture/T_Energy_Noise_01}.R
  )
  n4 = Fresnel(ExponentIn=Constant() {r=0.3}, BaseReflectFractionIn=Constant())
  n5 = Add(A=Saturate(None=Multiply(A=@n3) {constB=1.2}), B=Power(Base=OneMinus(None=@n4)))
  n6 = Add(A=TextureCoordinate(), B=AppendVector(A=Constant(), B=ParticleRandom()))
  n7 = Multiply(A=TextureSample(UVs=Panner(Coordinate=@n6) {speedY=0.5}) {texture=&T_Inky_Smoke_Tile}.R) {constB=5}
  n8 = ParticleColor()
  MaterialFunctionCall(Emissive=Multiply(A=Add(A=@n5, B=@n7), B=@n8.RGB)) {materialFunction=&MF_FX_EmissiveAdjustments}
  n9 = TextureSampleParameter2D(UVs=Panner(Coordinate=TextureCoordinate() {uTiling=2}) {speedX=-1}) {parameterName=MaskTexture, texture=/Game/Effects/Fort_Effects/Textures/T_Smoke_Ring_Mask2}
  n10 = TextureCoordinate()
  n11 = AppendVector(A=ScalarParameter() {defaultValue=1, parameterName=MaskU}, B=ScalarParameter() {defaultValue=1.2, parameterName=MaskV})
  n12 = TextureSample(UVs=Panner(Coordinate=Multiply(A=@n10, B=@n11)) {speedX=-0.5, speedY=0.3}) {texture=&T_Energy_Noise_01}
  n13 = Multiply(
   A=Multiply(A=@n8.A, B=Multiply(A=@n9.R, B=@n12.R))
   B=MaterialFunctionCall("FadeSensitivity (optional)"=ScalarParameter() {defaultValue=5, parameterName=MeshFade}) {materialFunction=&FadeBasedOnViewAngle}.Result
  )
  MaterialFunctionCall(Alpha Threshold=@n13) {materialFunction=&DitherTemporalAA_Fortnite}
  n14 = Panner(Coordinate=Add(A=TextureCoordinate(), B=ParticleRandom())) {speedX=0.5, speedY=1}
  Multiply(
   A=Multiply(
    A=TextureSample(UVs=@n14) {samplerType=SAMPLERTYPE_Grayscale, texture=&LowResBlurredNoise}.R
    B=ScalarParameter() {defaultValue=40, parameterName=WPO Intensity}
   ) {constB=25}
   B=VertexNormalWS()
  ) {constB=0}
  n15 = DepthFade(Opacity=@n13, FadeDistance=ScalarParameter() {defaultValue=5, parameterName=DepthFade})
  ShadingPathSwitch(Default=@n15, Deferred=@n15, Forward=@n13, Mobile=@n13)
 ]
}
