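# Translucent unlit Niagara material for animated energy-ball effects on mesh and sprite particles.
# Emissive sums panning smoke, directional and Voronoi noise, tinted by particle color, through an emissive adjustment function.
# Opacity combines scrolling noises, a parameterized mask, particle alpha and view-angle fade; depth fade applies only on default and deferred paths.
# A panning blurred-noise branch, randomized per particle and scaled along vertex normals, computes WPO-style displacement.
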Material /Game/Effects/Niagara/MasterMaterials/Energy/M_Energy_ball_02
graph {
 properties {
  bAutomaticallySetUsageInEditor=false
  bUsedWithNiagaraMeshParticles=true
  bUsedWithNiagaraSprites=true
  blendMode=BLEND_Translucent
  shadingModel=MSM_Unlit
  usedShadingModels=Unlit
 }
 nodes [
  n1 = Add(A=TextureCoordinate(), B=AppendVector(A=Constant(), B=ParticleRandom()))
  n2 = TextureCoordinate() {uTiling=0, vTiling=0}
  n3 = LinearInterpolate(Alpha=TextureSample(UVs=Panner(Coordinate=@n2)) {texture=/Game/Global/Textures/Default/Noise/Noise}.R) {constB=4}
  n4 = Multiply(
   A=TextureSample(UVs=Panner(Coordinate=@n1) {speedX=-0.8}) {texture=&T_Inky_Smoke_Tile}.R
   B=Multiply(A=Saturate(None=@n3)) {constB=5}
  )
  n5 = Panner(Coordinate=TextureCoordinate() {vTiling=2}) {speedX=-1, speedY=0.2}
  n6 = TextureSample(UVs=Panner(Coordinate=TextureCoordinate()) {speedX=-0.3, speedY=0.1}) {samplerType=SAMPLERTYPE_LinearGrayscale, texture=/Game/Packages/Fortress_Sky/SkyDome/Master/T_SkyDome_Cloud_Voroni}
  n7 = Add(
   A=Multiply(A=TextureSample(UVs=@n5) {samplerType=SAMPLERTYPE_LinearGrayscale, texture=/Game/Effects/Fort_Effects/Emotes/UltraLight/Texture/T_DirectionalNoise_Greyscale}.R) {constB=1.5}
   B=Multiply(A=LinearInterpolate(Alpha=@n6.R) {constA=-0.4}) {constB=0.3}
  )
  n8 = ParticleColor()
  MaterialFunctionCall(Emissive=Multiply(A=Add(A=@n4, B=@n7), B=@n8.RGB)) {materialFunction=&MF_FX_EmissiveAdjustments}
  n9 = TextureCoordinate()
  n10 = ParticleRandom()
  n11 = VectorParameter() {parameterName=WPOTexture, defaultValue=<rgba 1 1 0 0>}
  n12 = Panner(Coordinate=Multiply(A=Add(A=@n9, B=@n10), B=AppendVector(A=@n11.R, B=@n11.R))) {speedX=0.5, speedY=1}
  Multiply(
   A=Multiply(
    A=TextureSample(UVs=@n12) {samplerType=SAMPLERTYPE_Grayscale, texture=&LowResBlurredNoise}.R
    B=ScalarParameter() {defaultValue=30, parameterName=WPO Intensity}
   ) {constB=25}
   B=VertexNormalWS()
  ) {constB=0}
  n13 = Panner(Coordinate=TextureCoordinate() {vTiling=1.7}) {speedX=-0.5, speedY=0.4}
  n14 = Panner(Coordinate=TextureCoordinate()) {speedX=-0.8, speedY=0.6}
  n15 = LinearInterpolate(Alpha=Add(
    A=TextureSample(UVs=@n13) {texture=/Game/Global/Textures/Default/Noise/T_SoftCells_Negative}.R
    B=TextureSample(UVs=@n14) {texture=/Game/Effects/Niagara/MasterMaterials/Energy/Texture/T_Energy_Noise_01}.R
   )) {constA=1, constB=3.5}
  n16 = TextureSampleParameter2D(UVs=Panner(Coordinate=TextureCoordinate() {uTiling=2}) {speedX=-1}) {parameterName=MaskTexture, texture=/Game/Effects/Fort_Effects/Textures/T_Smoke_Ring_Mask2}
  n17 = Multiply(
   A=Multiply(A=@n8.A, B=Multiply(A=@n15, B=@n16.R))
   B=MaterialFunctionCall("FadeSensitivity (optional)"=ScalarParameter() {defaultValue=2.5, parameterName=MeshFade}) {materialFunction=&FadeBasedOnViewAngle}.Result
  )
  n18 = DepthFade(Opacity=@n17, FadeDistance=ScalarParameter() {defaultValue=10, parameterName=DepthFade})
  ShadingPathSwitch(Default=@n18, Deferred=@n18, Forward=@n17, Mobile=@n17)
 ]
}
