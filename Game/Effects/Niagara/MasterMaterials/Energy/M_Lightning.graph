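# Translucent unlit master material for Niagara mesh-particle lightning strips, tinted by particle color.
# Dynamic-parameter-offset noise, squared, modulates a soft plasma texture for emissive and also jitters vertices.
# World position offset builds camera-facing strips via a cross product of axis and view vector, widening with camera distance, plus a camera offset.
# Opacity combines plasma and particle alpha, adding depth fade only on the default shading path.
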Material /Game/Effects/Niagara/MasterMaterials/Energy/M_Lightning
graph {
 properties {
  bEnableResponsiveAA=true
  bUsedWithMeshParticles=true
  bUsedWithNiagaraMeshParticles=true
  blendMode=BLEND_Translucent
  pixelDepthOffsetMode=PDOM_Legacy
  shadingModel=MSM_Unlit
  translucencyPass=MTP_BeforeDOF
  twoSided=true
  usedShadingModels=Unlit
 }
 nodes [
  n1 = DynamicParameter() {defaultValue=<rgba 20 1 200 1>, paramNames [Stripwide,Random,WPO,Param4]}
  n2 = TextureSample(UVs=Add(A=TextureCoordinate() {vTiling=0.1}, B=@n1.Random)) {samplerType=SAMPLERTYPE_LinearColor, texture=&T_NoisePack_LinearColor}
  n3 = TextureSample(UVs=MaterialFunctionCall() {materialFunction=&Swizzle}.YX) {texture=/Game/Effects/Fort_Effects/Textures/Tile/T_Lightning_Plasma_Soft}
  n4 = ParticleColor()
  MaterialFunctionCall(Emissive=Multiply(A=Multiply(A=@n2.G, B=@n2.G), B=Multiply(A=@n3.R, B=@n4.RGB))) {materialFunction=&MF_FX_EmissiveAdjustments}
  n5 = Constant3Vector() {constant=<rgba 1 0 0 0>}
  n6 = Multiply(A=Subtract(A=@n2.RGB) {constB=0.5}, B=@n1.WPO)
  n7 = Subtract(A=Add(A=WorldPosition().XYZ, B=@n6), B=CameraPositionWS())
  n8 = Distance(A=CameraPositionWS(), B=ObjectPositionWS())
  n9 = LinearInterpolate(Alpha=Saturate(None=Divide(A=@n8) {constB=10000})) {constA=0.2, constB=3}
  n10 = ComponentMask(None=TextureCoordinate()) {g=true}
  n11 = Multiply(
   A=CrossProduct(A=Transform(None=@n5), B=Normalize(VectorInput=@n7))
   B=Multiply(A=Multiply(A=@n1.Stripwide, B=@n9), B=Subtract(A=@n10) {constB=0.5}) {constB=200}
  )
  n12 = Constant() {r=500}
  Reroute(None=Add(
    A=Add(A=@n11, B=@n6)
    B=MaterialFunctionCall(Offset Amount=@n12) {materialFunction=/Engine/Functions/Engine_MaterialFunctions02/WorldPositionOffset/CameraOffset}.Clamped Camera Offset
   ))
  n13 = DepthFade() {fadeDistanceDefault=200}
  n14 = Multiply(A=@n3.R, B=@n4.A)
  ShadingPathSwitch(Default=ShadingPathSwitch(Default=Multiply(A=@n13, B=@n14), Mobile=@n14) {desc=Migrated from FeatureLevelSwitch}, Forward=@n14, Mobile=@n14)
 ]
}
