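# Unlit, two-sided translucent Niagara master material for sprite and mesh particles, named for Taker portal upstreaks.
# Two tiled textures with time-driven offsets combine, the second weighted by a vertical gradient.
# A mask of bottom and top UV fades with optional Fresnel multiplies or adds to that noise, shaping depth-faded opacity scaled by particle alpha.
# Particle color drives adjusted emissive, while the mask scales refraction and the second texture builds a distortion normal.
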Material /Game/Effects/Niagara/MasterMaterials/Energy/M_TakerPortalUpstreaks_BR
graph {
 properties {
  bUsedWithNiagaraMeshParticles=true
  bUsedWithNiagaraSprites=true
  blendMode=BLEND_Translucent
  pixelDepthOffsetMode=PDOM_Legacy
  refractionMethod=RM_IndexOfRefraction
  shadingModel=MSM_Unlit
  twoSided=true
  usedShadingModels=Unlit
 }
 nodes [
  n1 = ParticleColor()
  n2 = TextureCoordinate()
  n3 = VectorParameter() {parameterName=Texture01UV, defaultValue=<rgba 2 0.25 0 0.8>}
  n4 = AppendVector(A=@n3.R, B=@n3.R)
  n5 = AppendVector(A=@n3.B, B=@n3.B)
  n6 = LinearInterpolate(A=Time(), Alpha=ScalarParameter() {parameterName="Texture01UVPanner=0"})
  n7 = TextureSampleParameter2D(UVs=Add(A=Multiply(A=@n2, B=@n4), B=Multiply(A=@n5, B=@n6))) {parameterName=Texture01, texture=&T_ShieldCaustics}
  n8 = TextureCoordinate()
  n9 = VectorParameter() {parameterName=Texture02UV, defaultValue=<rgba 3 0.5 0 0.35>}
  n10 = AppendVector(A=@n9.R, B=@n9.R)
  n11 = AppendVector(A=@n9.B, B=@n9.B)
  n12 = LinearInterpolate(A=Time(), Alpha=ScalarParameter() {parameterName="Texture02UVPanner=0"})
  n13 = TextureSampleParameter2D(UVs=Add(A=Multiply(A=@n8, B=@n10), B=Multiply(A=@n11, B=@n12))) {parameterName=Texture02, texture=&T_NoisePack_01}
  n14 = ChannelMaskParameter(None=@n13.RGB) {parameterName=Texture02Channel}
  n15 = ComponentMask(None=TextureCoordinate()) {g=true}
  n16 = ScalarParameter() {defaultValue=1, parameterName=MaskUpInt}
  n17 = Power(Base=Saturate(None=Multiply(A=@n15, B=@n16)), Exp=ScalarParameter() {defaultValue=2, parameterName=MaskUpPower})
  n18 = Saturate(None=Add(A=ChannelMaskParameter(None=@n7.RGB) {parameterName=Texture01Channel}.R, B=Multiply(A=@n14.R, B=@n17) {constB=0.8}))
  n19 = MaterialFunctionCall(Exponent=ScalarParameter() {defaultValue=4, parameterName=FresnelExponent}) {materialFunction=&MF_Fresnel_FX}
  n20 = ScalarParameter() {defaultValue=1, parameterName=UseFresnel}
  n21 = TextureCoordinate()
  n22 = Multiply(A=OneMinus(None=ComponentMask(None=@n21) {g=true}), B=ScalarParameter() {defaultValue=8, parameterName=MaskDownInt}) {constB=8}
  n23 = Multiply(
   A=Multiply(A=@n17, B=LinearInterpolate(B=@n19.Fresnel, Alpha=@n20) {constA=1})
   B=Power(Base=Saturate(None=@n22), Exp=ScalarParameter() {defaultValue=2, parameterName=MaskDownPower})
  )
  n24 = LinearInterpolate(A=Multiply(A=@n18, B=@n23), B=Add(A=@n18, B=@n23), Alpha=ScalarParameter() {parameterName=Add Mask Noise?})
  n25 = ScalarParameter() {defaultValue=1, parameterName=OpacityExp}
  DepthFade(Opacity=Multiply(A=@n1.A, B=Power(Base=@n24, Exp=@n25)), FadeDistance=ScalarParameter() {parameterName=DepthFade})
  LinearInterpolate(B=ScalarParameter() {defaultValue=1.1, parameterName=Refraction}, Alpha=Multiply(A=@n1.A, B=@n23)) {constA=1}
  n26 = MaterialFunctionCall(Float2=ConstantBiasScale(None=MaterialFunctionCall(X=@n13.G, Y=@n13.G) {materialFunction=&MakeFloat2}.Result) {bias=-0.5, scale=2}) {materialFunction=&BreakOutFloat2Components}
  MaterialFunctionCall(X=@n26.R, Y=@n26.R, Z=Constant() {r=1}) {materialFunction=&MakeFloat3}
  Reroute(None=MaterialFunctionCall(Emissive=@n1.RGB) {materialFunction=&MF_FX_EmissiveAdjustments}.Result)
 ]
}
